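# Material function for the Heatwave item shop UI that generates animated procedural background noise from UV and time, without textures.
# Three sine layers use the R, G and B channels of the horizontal and vertical tiling inputs: the first adds a time-driven rotator, the other two use skewed UVs scrolled in differing directions.
# The layers are offset and summed, then their vector length is scaled and squared into a soft scalar output.
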MaterialFunction /ItemShop/UI/Art/Heatwave/MF_UI_BackgroundNoise
graph {nodes [
  AppendVector()
  n1 = FunctionInput() {bUsePreviewValueAsDefault=true, inputName=HorizontalTiling, previewValue=<0.5 2.3 1.5 1>}
  n2 = ComponentMask(None=@n1) {r=true}
  n3 = FunctionInput() {bUsePreviewValueAsDefault=true, inputName=VerticalTiling, previewValue=<0.6 1.5 2.3 1>}
  n4 = ComponentMask(None=@n3) {r=true}
  n5 = Multiply(A=FunctionInput(Preview=TextureCoordinate()) {bUsePreviewValueAsDefault=true, inputName=UV, inputType=FunctionInput_Vector2})
  n6 = FunctionInput(Preview=Time()) {bUsePreviewValueAsDefault=true, inputName=Time, inputType=FunctionInput_Scalar}
  n7 = Add(A=Multiply(A=@n5, B=AppendVector(A=@n2, B=@n4)), B=Rotator(Time=Multiply(A=@n6) {desc=speed offset}))
  n8 = ComponentMask(None=@n5) {r=true}
  n9 = ComponentMask(None=@n5) {g=true}
  n10 = Add(A=@n8, B=Multiply(A=@n9) {constB=-0.3})
  n11 = ComponentMask(None=@n1) {g=true}
  n12 = ComponentMask(None=@n3) {g=true}
  n13 = Add(
   A=Multiply(A=AppendVector(A=@n10, B=@n9), B=AppendVector(A=@n11, B=@n12))
   B=Multiply(A=@n6, B=Constant2Vector() {g=0.25, r=-0.2}) {desc=speed offset}
  )
  n14 = Add(A=Add(A=Sine(None=@n7), B=Sine(None=@n13)), B=Constant2Vector() {g=0.15, r=0.2})
  n15 = Add(A=ComponentMask(None=@n14) {r=true}, B=ComponentMask(None=@n14) {g=true})
  n16 = Multiply(A=@n8) {constB=0.3}
  n17 = Multiply(A=AppendVector(A=@n8, B=Add(A=@n16, B=@n9)), B=AppendVector(A=ComponentMask(None=@n1) {b=true}, B=ComponentMask(None=@n3) {b=true}))
  n18 = Multiply(A=@n6, B=Constant2Vector() {g=-0.1, r=0.25}) {desc=speed offset}
  n19 = Add(A=Sine(None=Add(A=@n17, B=@n18)), B=Constant2Vector() {g=0.25, r=0.3})
  n20 = Multiply(A=Length(None=Add(A=@n15, B=@n19)) {is=class MaterialExpressionLength}) {constB=0.1}
  FunctionOutput(None=Multiply(A=@n20, B=@n20) {constB=0.3})
 ]}
