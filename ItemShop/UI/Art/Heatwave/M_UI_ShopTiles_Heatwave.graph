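# Translucent UI material that draws the animated Heatwave-themed background for Item Shop tiles.
# Side A color-ramps rotated waves built from scrolling noise and a sine term, while side B color-ramps a rotated, scaled radial gradient.
# A skewed, eased elliptical mask blends the two sides through named reroutes.
# An Overlay-blended vignette tint, Color Dodge film grain and a rounded-corner function provide the finishing layers.
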Material /ItemShop/UI/Art/Heatwave/M_UI_ShopTiles_Heatwave
graph {
 properties {bAutomaticallySetUsageInEditor=false, blendMode=BLEND_Translucent, materialDomain=MD_UI, usedShadingModels=Default Lit}
 nodes [
  Constant() {r=1}
  Constant() {r=1}
  n1 = NamedRerouteUsage()
  n2 = VectorParameter() {group=Vignette, parameterName=VignetteColor, defaultValue=<rgba 1 0.2119 0.0298 0.5>}
  n3 = MaterialFunctionCall(Base=@n1, Blend=@n2.RGB) {materialFunction=&Blend_Overlay}
  n4 = Constant() {r=1}
  n5 = ScalarParameter() {defaultValue=-0.5, group=Vignette, parameterName=VignettePositionX}
  n6 = ScalarParameter() {defaultValue=-0.9, group=Vignette, parameterName=VignettePositionY}
  n7 = Multiply(
   A=Add(A=NamedRerouteUsage(), B=AppendVector(A=@n5, B=@n6))
   B=AppendVector(
    A=ScalarParameter() {defaultValue=0.5, group=Vignette, parameterName=VignetteSharpnessX}
    B=ScalarParameter() {defaultValue=1.4, group=Vignette, parameterName=VignetteSharpnessY}
   )
  )
  n8 = Saturate(None=Subtract(
    A=Length(None=@n7) {is=class MaterialExpressionLength}
    B=ScalarParameter() {defaultValue=0.3, group=Vignette, parameterName=VignetteFillSize}
   ))
  n9 = Multiply(A=@n8, B=@n8)
  MaterialFunctionCall!(
   Blend=Multiply(
    A=MaterialFunctionCall!(UV, Seed) {materialFunction=&MF_UI_Random}.Result
    B=ScalarParameter() {defaultValue=0.22, group=Film Grain, parameterName=Film Noise Opacity, sliderMax=1}
   ) {constA=0.15, constB=0.15}
   Base=LinearInterpolate(A=@n1, B=AppendVector(A=@n3.Result, B=@n4), Alpha=Multiply(A=@n2.A, B=@n9)) {constAlpha=1}
  ) {materialFunction=/Engine/Functions/Engine_MaterialFunctions03/Blends/Blend_ColorDodge}
  MaterialFunctionCall!(Input Alpha=Constant() {r=1}, bOverrideCornerRadius, Corner Radius Override) {materialFunction=/FortUILibrary/Systems/ItemShopTile/Materials/MaterialFunctions/MF_UI_ItemShop_RoundedCorners}
  MaterialFunctionCall!(Use Local UV, Texture Size, Scale Pivot Point) {materialFunction=/Game/Environments/FrontEnd/Blueprints/ItemPreview/VaultWorldTransitions/MaterialFunctions/MF_UI_VaultWorld_UV}
  n10 = Add(
   A=NamedRerouteUsage()
   B=AppendVector(
    A=ScalarParameter() {defaultValue=0.5, group=SideA, parameterName=WavesPositionHoriz}
    B=ScalarParameter() {defaultValue=-0.9, group=SideA, parameterName=WavesPositionVert}
   )
  ) {constB=0.5}
  n11 = Constant() {r=-0.6}
  n12 = OneMinus(None=ComponentMask(None=MaterialFunctionCall(UVs=@n10, "Rotation Angle (0-1)"=@n11) {materialFunction=&CustomRotator}.Rotated Values) {g=true})
  Multiply(
   A=Add(A=@n12, B=ScalarParameter() {defaultValue=-0.1, group=SideA, parameterName=WavesAngleLanding}) {constB=-0.95, desc=angle}
   B=ScalarParameter() {group=SideA, parameterName=WavesSharpnessLanding}
  ) {constB=10, desc=sharpness}
  NamedRerouteDeclaration(None=MaterialFunctionCall!(ScreenUV, TextureSize, ScalePivotPoint) {materialFunction=/FortUILibrary/MaterialFunctions/Basics/MF_UI_ScreenSpaceUVScaling}.ScaleY) {name=Scale Y, nodeColor=<rgba 0 0 0 1>}
  n13 = Time()
  n14 = ScalarParameter() {defaultValue=0.5, group=SideA, parameterName=WavesRandomSpeed}
  n15 = NamedRerouteUsage()
  n16 = Constant2Vector() {desc=scale, g=1, r=0.4}
  n17 = ComponentMask(None=Rotator(
    Coordinate=MaterialFunctionCall(UVs=@n15, Texture Scale=@n16) {materialFunction=&ScaleUVsByCenter}.UVs
    Time=Constant() {r=-14}
   )) {r=true}
  n18 = OneMinus(None=MaterialFunctionCall!(HorizontalTiling, VerticalTiling, Time=Multiply(A=@n13, B=@n14) {constB=0.2, desc=pattern speed}, UV=@n17) {materialFunction=&/MF_UI_BackgroundNoise}.Result)
  n19 = Multiply(A=@n18, B=@n18) {constB=2}
  n20 = ComponentMask(None=@n17) {g=true}
  n21 = OneMinus(None=Multiply(A=@n20, B=@n20) {constB=0.5})
  n22 = Multiply(A=Add(A=Multiply(A=@n21, B=@n17) {constB=0.5}) {constB=-0.03}) {constB=5}
  n23 = NamedRerouteDeclaration(None=Multiply(A=OneMinus(None=@n19), B=Saturate(None=@n22))) {name=WavePattern, nodeColor=<rgba 0 0.9062 1 1>}
  n24 = ScalarParameter() {defaultValue=0.5, group=SideA, parameterName=WavesRandomSize}
  n25 = Time()
  n26 = ScalarParameter() {defaultValue=0.05, group=SideA, parameterName=WavesConstantSpeed}
  n27 = Sine(None=Add(A=@n17, B=Multiply(A=@n25, B=@n26) {constB=0.05, desc=constant waves speed}))
  n28 = Multiply(A=@n27, B=@n27) {constB=0.5}
  n29 = Multiply(A=@n28, B=@n28) {constB=0.5}
  n30 = ScalarParameter() {defaultValue=0.2, group=SideA, parameterName=WavesConstantSize}
  n31 = Add(A=@n12, B=Add(A=Multiply(A=@n23, B=@n24) {constB=0.2}, B=Multiply(A=@n29, B=@n30) {constB=0.07}) {constB=0})
  n32 = OneMinus(None=MaterialFunctionCall("Value (0-1)"=Saturate(None=@n31)) {materialFunction=&MF_UI_EaseCurves}.EaseInOutCubic)
  NamedRerouteDeclaration(None=Add(
    A=ColorRamp!(None=@n32) {colorCurve="<self>.ColorCurve"}
    B=VectorParameter() {group=SideA, parameterName=SideAColorScreen, defaultValue=<rgba 0 0.007 0.0948 1>}.RGBA
   ) {constB=0.02}) {name=SideABlend, nodeColor=<rgba 0 0 0 1>}
  n33 = ScalarParameter() {defaultValue=0.8, group=SideB, parameterName=SideBBrightness}
  n34 = Add(A=NamedRerouteUsage(), B=Constant2Vector() {g=0.1, r=0.2})
  n35 = Constant() {r=0.4}
  n36 = MaterialFunctionCall(UV=MaterialFunctionCall(
    UVs=MaterialFunctionCall(UVs=@n34, "Rotation Angle (0-1)"=@n35) {materialFunction=&CustomRotator}.Rotated Values
    Texture Scale=Constant2Vector() {g=1.25, r=2.5}
   ) {materialFunction=&ScaleUVsByCenter}.UVs) {materialFunction=&MF_UI_RadialGradient}
  NamedRerouteDeclaration(None=ColorRamp(None=Multiply(A=@n36.Smooth, B=@n33) {constB=0.8}) {colorCurve="<self>.ColorCurve"}) {name=SideBBlend, nodeColor=<rgba 0 0 0 1>}
  n37 = NamedRerouteUsage()
  n38 = ComponentMask(None=@n37) {r=true}
  n39 = ScalarParameter() {defaultValue=-0.6, group=SidesMask, parameterName=SidesMaskVerticalSkew}
  n40 = AppendVector(A=@n38, B=Add(A=Multiply(A=@n38, B=@n39) {constB=-0.6, desc=skew}, B=ComponentMask(None=@n37) {g=true}))
  n41 = AppendVector(
   A=ScalarParameter() {defaultValue=0.9, group=SideA, parameterName=SidesMaskPositionX}
   B=ScalarParameter() {defaultValue=-0.93, group=SideA, parameterName=SidesMaskPositionY}
  )
  n42 = ScalarParameter() {defaultValue=0.3, group=SideA, parameterName=SidesMaskSharpnessX}
  n43 = ScalarParameter() {defaultValue=0.9, group=SideA, parameterName=SidesMaskSharpnessY}
  n44 = Length(None=Multiply(A=Add(A=@n40, B=@n41), B=AppendVector(A=@n42, B=@n43)))
  n45 = ScalarParameter() {defaultValue=0.3, group=SidesMask, parameterName=SidesMaskFillSize}
  n46 = Saturate(None=Multiply(A=Subtract(A=@n44, B=@n45), B=ScalarParameter() {defaultValue=2.5, group=SidesMask, parameterName=SidesMaskSharpness}) {constB=2.5})
  NamedRerouteDeclaration(None=LinearInterpolate(
    A=NamedRerouteUsage()
    B=NamedRerouteUsage()
    Alpha=MaterialFunctionCall("Value (0-1)"=@n46) {materialFunction=&MF_UI_EaseCurves}.EaseInOutSine
   )) {name=SidesBlend, nodeColor=<rgba 0 0 0 1>}
  NamedRerouteDeclaration(None=MaterialFunctionCall() {materialFunction=&MF_UI_UVScaling}.ScaleY) {name=UVScale_Y, nodeColor=<rgba 0.5 0 1 1>}
  OneMinus(None=@n10)
 ]
}
